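# Physics asset for the GenericHair10 face accessory part of the BlockStack medium male body, setting up physical motion for six hair strands.
# Each strand, HairA through HairF, is a chain of three capsule dyn bodies hinged from a tiny base capsule by swing-limited constraints with little or no twist.
# A and F share longer capsules, while B through E use short, round ones.
# A large box on deform_spine_05, named as a head box, has no constraints.
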PhysicsAsset /BRCosmetics/Characters/Player/Male/Medium/Bodies/M_MED_BlockStack_GenericHair10/Meshes/Parts/M_MED_BlockStack_GenericHair10_FaceAcc_Physics
bodies {
 HairA_01_base [{shapeType=Capsule, length=0.505, radius=0.505, shapeName=HairA_01_base_capsule, rotation=<rot 0 -90 0>}]
 HairA_01_dyn [{shapeType=Capsule, length=5, radius=1.75, shapeName=HairA_01_dyn_capsule, center=<2 0 0>, rotation=<rot 90 0 0>}]
 HairA_02_dyn [{shapeType=Capsule, length=5, radius=1.1, shapeName=HairA_02_dyn_capsule, center=<2 0 1>, rotation=<rot 90 0 0>}]
 HairA_03_dyn [{shapeType=Capsule, length=4, radius=1.6, shapeName=HairA_03_dyn_capsule, center=<1.2664 0 0>, rotation=<rot 90 0 0>}]
 HairB_01_base [{shapeType=Capsule, length=0.505, radius=0.505, shapeName=HairB_01_base_capsule, rotation=<rot 0 -90 0>}]
 HairB_01_dyn [{shapeType=Capsule, length=2, radius=2, shapeName=HairB_01_dyn_capsule, center=<0.3987 0 0>, rotation=<rot 90 0 0>}]
 HairB_02_dyn [{shapeType=Capsule, length=2, radius=2, shapeName=HairB_02_dyn_capsule, center=<2 0 0>, rotation=<rot 90 0 0>}]
 HairB_03_dyn [{shapeType=Capsule, length=2, radius=2, shapeName=HairB_03_dyn_capsule, center=<2.5 0 0>, rotation=<rot 90 0 0>}]
 HairC_01_base [{shapeType=Capsule, length=0.505, radius=0.505, shapeName=HairC_01_base_capsule, rotation=<rot 0 -90 0>}]
 HairC_01_dyn [{shapeType=Capsule, length=2, radius=2, shapeName=HairC_01_dyn_capsule, center=<0.5683 0 0>, rotation=<rot 90 0 0>}]
 HairC_02_dyn [{shapeType=Capsule, length=2, radius=2, shapeName=HairC_02_dyn_capsule, center=<2 0 0>, rotation=<rot 90 0 0>}]
 HairC_03_dyn [{shapeType=Capsule, length=2, radius=2, shapeName=HairC_03_dyn_capsule, center=<3.5 0 0>, rotation=<rot 90 0 0>}]
 HairD_01_base [{shapeType=Capsule, length=0.505, radius=0.505, shapeName=HairD_01_base_capsule, rotation=<rot 0 -90 0>}]
 HairD_01_dyn [{shapeType=Capsule, length=2, radius=2, shapeName=HairD_01_dyn_capsule, rotation=<rot 90 0 0>}]
 HairD_02_dyn [{shapeType=Capsule, length=2, radius=2, shapeName=HairD_02_dyn_capsule, center=<2 0 0>, rotation=<rot 90 0 0>}]
 HairD_03_dyn [{shapeType=Capsule, length=2, radius=2, shapeName=HairD_03_dyn_capsule, center=<3.5 0 0>, rotation=<rot 90 0 0>}]
 HairE_01_base [{shapeType=Capsule, length=0.505, radius=0.505, shapeName=HairE_01_base_capsule, rotation=<rot 0 -90 0>}]
 HairE_01_dyn [{shapeType=Capsule, length=2, radius=2, shapeName=HairE_01_dyn_capsule, center=<0.7472 0 0>, rotation=<rot 90 0 0>}]
 HairE_02_dyn [{shapeType=Capsule, length=2, radius=2, shapeName=HairE_02_dyn_capsule, center=<1.7472 0 0>, rotation=<rot 90 0 0>}]
 HairE_03_dyn [{shapeType=Capsule, length=2, radius=2, shapeName=HairE_03_dyn_capsule, center=<3 0 0>, rotation=<rot 90 0 0>}]
 HairF_01_base [{shapeType=Capsule, length=0.505, radius=0.505, shapeName=HairF_01_base_capsule, rotation=<rot 0 -90 0>}]
 HairF_01_dyn [{shapeType=Capsule, length=5, radius=1.75, shapeName=HairF_01_dyn_capsule, center=<2 0 0>, rotation=<rot 90 0 0>}]
 HairF_02_dyn [{shapeType=Capsule, length=5, radius=1.1, shapeName=HairF_02_dyn_capsule, center=<2 0 1>, rotation=<rot 90 0 0>}]
 HairF_03_dyn [{shapeType=Capsule, length=4, radius=1.6, shapeName=HairF_03_dyn_capsule, center=<1 0 0>, rotation=<rot 90 0 0>}]
 deform_spine_05 [{shapeType=Box, extentX=34.2204, extentY=22.2631, extentZ=31.349, shapeName=deform_head_box, center=<3.9963 0.7091 0>}]
}
constraints [
 {bone1Name=HairA_01_dyn, bone2Name=HairA_01_base, swing1LimitDegrees=15, swing2LimitDegrees=25, twistLimitDegrees=2}
 {bone1Name=HairA_02_dyn, bone2Name=HairA_01_dyn, swing1LimitDegrees=15, swing2LimitDegrees=25, twistLimitDegrees=2}
 {bone1Name=HairA_03_dyn, bone2Name=HairA_02_dyn, swing1LimitDegrees=15, swing2LimitDegrees=25, twistLimitDegrees=0}
 {bone1Name=HairB_01_dyn, bone2Name=HairB_01_base, swing1LimitDegrees=10, swing2LimitDegrees=40, twistLimitDegrees=0}
 {bone1Name=HairB_02_dyn, bone2Name=HairB_01_dyn, swing1LimitDegrees=15, swing2LimitDegrees=25, twistLimitDegrees=0}
 {bone1Name=HairB_03_dyn, bone2Name=HairB_02_dyn, swing1LimitDegrees=15, swing2LimitDegrees=15, twistLimitDegrees=0}
 {bone1Name=HairC_01_dyn, bone2Name=HairC_01_base, swing1LimitDegrees=10, swing2LimitDegrees=40, twistLimitDegrees=0}
 {bone1Name=HairC_02_dyn, bone2Name=HairC_01_dyn, swing1LimitDegrees=15, swing2LimitDegrees=25, twistLimitDegrees=0}
 {bone1Name=HairC_03_dyn, bone2Name=HairC_02_dyn, swing1LimitDegrees=15, swing2LimitDegrees=15, twistLimitDegrees=0}
 {bone1Name=HairD_01_dyn, bone2Name=HairD_01_base, swing1LimitDegrees=10, swing2LimitDegrees=40, twistLimitDegrees=0}
 {bone1Name=HairD_02_dyn, bone2Name=HairD_01_dyn, swing1LimitDegrees=15, swing2LimitDegrees=25, twistLimitDegrees=0}
 {bone1Name=HairD_03_dyn, bone2Name=HairD_02_dyn, swing1LimitDegrees=15, swing2LimitDegrees=15, twistLimitDegrees=0}
 {bone1Name=HairE_01_dyn, bone2Name=HairE_01_base, swing1LimitDegrees=10, swing2LimitDegrees=40, twistLimitDegrees=0}
 {bone1Name=HairE_02_dyn, bone2Name=HairE_01_dyn, swing1LimitDegrees=15, swing2LimitDegrees=25, twistLimitDegrees=0}
 {bone1Name=HairE_03_dyn, bone2Name=HairE_02_dyn, swing1LimitDegrees=15, swing2LimitDegrees=15, twistLimitDegrees=0}
 {bone1Name=HairF_01_dyn, bone2Name=HairF_01_base, swing1LimitDegrees=15, swing2LimitDegrees=25, twistLimitDegrees=2}
 {bone1Name=HairF_02_dyn, bone2Name=HairF_01_dyn, swing1LimitDegrees=15, swing2LimitDegrees=25, twistLimitDegrees=2}
 {bone1Name=HairF_03_dyn, bone2Name=HairF_02_dyn, swing1LimitDegrees=15, swing2LimitDegrees=25, twistLimitDegrees=0}
]
