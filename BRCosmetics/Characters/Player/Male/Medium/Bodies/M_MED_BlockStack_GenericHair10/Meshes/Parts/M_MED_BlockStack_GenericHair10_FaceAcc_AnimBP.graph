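# Post-process style AnimBlueprint for the face accessory part of the BlockStack GenericHair10 male character, adding secondary motion.
# The pose copied from the attached parent runs through a Control Rig, then a RigidBody simulation in deform_head space.
# Updates set the rig's isJumping input when jumping, falling or moving in 2D.
# They also compute gravity, RBAN settings and crouch/sprint values, which the AnimGraph never reads; its RigidBody gravity and settings are fixed.
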AnimBlueprint /BRCosmetics/Characters/Player/Male/Medium/Bodies/M_MED_BlockStack_GenericHair10/Meshes/Parts/M_MED_BlockStack_GenericHair10_FaceAcc_AnimBP
target_skeleton=&/M_MED_BlockStack_GenericHair10_FaceAcc_Skeleton
graphs [
 AnimGraph {nodes [
   n1 = Control Rig!(
    Source=Copy Pose From Mesh(SourceMeshComponent=Get Head() -> (Head:Skeletal Mesh Component Budgeted Object Reference)) {bUseAttachedParent=true}
    Alpha
    isJumping:bool=Get JumpingVar() -> (JumpingVar:bool)
   ) -> (Pose) {controlRigAssetReference {controlRigAsset=&/M_MED_BlockStack_GenericHair10_FaceAcc_CtrlRig}}
   n2 = RigidBody!(
    ComponentPose=Local To Component(LocalPose=@n1)
    Alpha
    OverrideWorldGravity:Vector="0,0,-980"
    ExternalForce
    SimSpaceSettings:Sim Space Settings Structure=(WorldAlpha=1,VelocityScaleZ=1,DampingAlpha=1,MaxLinearVelocity=200,MaxAngularVelocity=20,MaxLinearAcceleration=350,MaxAngularAcceleration=45,ExternalLinearDragV=(X=3,Y=3,Z=3),ExternalLinearVelocity=(X=0,Y=0,Z=0),ExternalAngularVelocity=(X=0,Y=0,Z=0))
   ) -> (Pose) {
    bOverrideWorldGravity=true
    overridePhysicsAsset=&/M_MED_BlockStack_GenericHair10_FaceAcc_Physics
    simulationSpace=BaseBoneSpace
    baseBoneRef {boneName=deform_head}
    overrideWorldGravity {z=-980}
    simSpaceSettings {
     maxAngularAcceleration=45
     maxAngularVelocity=20
     maxLinearAcceleration=350
     maxLinearVelocity=200
     worldAlpha=1
     externalLinearDragV=<3 3 3>
    }
   }
   Output Pose(Result=Component To Local(ComponentPose=@n2)) {layerGroup=None}
  ]}
 EventGraph {nodes [
   Event BlueprintUpdateAnimation()
   > n1 = Sequence!(execute) -> (then_0, then_1, then_2:Exec)
   > n2 = "Gravity Override State Function (Parameter Asset) (Nativized)"(
    CharacterPartAnimInstance=Self-Reference()
    GameTimeInSeconds=GetGameTimeInSeconds()
    DebugJointName=spine_05
    Parameters=/BRCosmetics/Characters/Player/Male/Medium/Bodies/M_MED_Blockstack_HopeExtraExtra/Meshes/Parts/F_MED_BlockStack_HopeExtraExtra_Parameters
   )
   > n3 = Branch(Condition=OR Boolean!(A=Get bIsJumping(), B=Get bIsFalling(), C:bool=Get bIsMoving2D()) -> (ReturnValue))
   > Set GravityOverride(GravityOverride="0,-500,-650")
   Set GravityOverride(execute=@n3.else, GravityOverride=@n2)
   > Set RBAN Settings Out(RBAN Settings Out=@n2.OutRigidBodyAnimNodeParameters)
   n4 = AND Boolean(A=Get bIsCrouching(), B=Get bIsMoving2D())
   n5 = AND Boolean(A=OR Boolean!(A=Get bIsJumping(), B=Get bIsFalling()) -> (ReturnValue), B=NOT Boolean(A=Get bIsSkyDiving()))
   n6 = Get bIsSprinting()
   n7 = Get bIsTacticalSprinting()
   Set JumpingAndTacticalSprinting(
    execute=@n1.then_1
    JumpingAndTacticalSprinting:double="To Float (Boolean)"(InBool=OR Boolean!(A=@n7, B=@n5, C:bool=@n4, D:bool=@n6) -> (ReturnValue))
   ) -> (Output_Get:double)
   n8 = NOT Boolean(A=Get bIsMoving2D())
   n9 = Get bIsCrouching()
   > Set CrouchingIdle(CrouchingIdle:double="To Float (Boolean)"(InBool=AND Boolean(A=@n9, B=@n8))) -> (Output_Get:double)
   n10 = Branch(execute=@n1.then_2, Condition=OR Boolean!(A=Get bIsJumping(), B=Get bIsFalling(), C=Get bIsMoving2D()) -> (ReturnValue))
   > Set JumpingVar(JumpingVar:bool=true) -> (Output_Get:bool)
   Set JumpingVar(execute=@n10.else, JumpingVar=false)
   Set GravityOverride(GravityOverride="0,-400,-650")
  ]}
]
variables [
 CrouchingIdle {type=number}
 GravityOverride {type=Vector}
 Head {type=SkeletalMeshComponentBudgeted}
 JumpingAndTacticalSprinting {type=number}
 JumpingVar {type=boolean}
 RBAN Settings Out {type=FortRigidBodyAnimNodeParameters}
]
